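; Parent gameplay ability for consuming foraged item versions of consumables in Athena.
; Activation holds only a disabled parent-call placeholder, so that event adds no logic of its own.
; When the ability ends without cancellation and has authority, it sends an Event.Forageable.Consumed gameplay event to the activating pawn.
; The end event then calls the parent implementation, passing the cancelled flag through.
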(event Ability|EventActivateAbility
  (|CallParentFunction) ; This node is disabled and will not be called.\nDrag off pins to build functionality.
)

(event Ability|EventOnEndAbility
  (params (bWasCancelled :type Boolean))
  (if (not bWasCancelled)
    (Ability|HasAuthority
      (:True
        (Ability|SendGameplayEventtoActor (Ability|GetActivatingPawn) "(TagName=\"Event.Forageable.Consumed\")"))
      (:False
        _)))
  (|CallParentFunction bWasCancelled))
